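# Animation Blueprint for the HammerVice Universal offhand pickaxe mesh, supplying its dual-wield idle pose.
# A single-state Idle Pose machine plays the HammerViceUniversal_Idle_OffHand sequence.
# A linear Blend Poses by bool selects between the linked input pose (true) and that idle machine (false); its bool is unwired at default false, so idle plays.
# The EventGraph holds only unconnected stub nodes, with no variables or logic.
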AnimBlueprint /BRCosmetics/Weapons/FORT_Melee/Pickaxe_HammerVice_UniversalOffhand/Meshes/HammerVice_UniversalOffhand_Axe_AnimBP
target_skeleton=&/HammerVice_UniversalOffhand_Axe_Skeleton
graphs [
 AnimGraph {nodes [
   AnimGraphNode_StateMachine_1 = Idle Pose() {is=title AnimGraphNode_StateMachine}
   AnimGraphNode_SubInput_0 = Input Pose()
   Output Pose(Result=Blend Poses by bool(BlendPose_0=@AnimGraphNode_SubInput_0, BlendPose_1=@AnimGraphNode_StateMachine_1) {blendType=Linear}) {layerGroup=None}
  ]}
 Idle Pose {entry=Idle, path=AnimGraph.AnimGraphNode_StateMachine_1.Idle Pose, ids=id, nodes [AnimStateNode_0 = Idle()]}
 Idle {
  path=AnimGraph.AnimGraphNode_StateMachine_1.Idle Pose.AnimStateNode_0.Idle
  nodes [Output Animation Pose(Result=Sequence Player 'HammerViceUniversal_Idle_OffHand'() {sequence=/BRCosmetics/Animation/Game/MainPlayer/Combat/Melee/DualWield/HammerVice_Universal/HammerViceUniversal_Idle_OffHand})]
 }
 EventGraph {nodes [
   TryGetPawnOwner()
   Event BlueprintUpdateAnimation()
  ]}
]
variables []
